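# Material function in the Visceral Tier2 wheel cosmetics that outputs camera-position and depth-based masks, plus a debug gradient.
# CameraPositionMask clamps the camera's position along a local forward vector, measured from an offset object center.
# A local-space Y gradient, shaped by Offset and Falloff, blends two depth biases.
# DepthGradientMask compares biased pixel depth to scene depth, zeroing where it lies farther, to remove nearby objects.
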MaterialFunction /VehicleCosmetics/Wheels/Visceral_Tier2/Materials/MF_CylinderVMask
graph {nodes [
  n1 = FunctionInput(Preview=Constant3Vector() {constant=<rgba 0 1 0 0>}) {bUsePreviewValueAsDefault=true, inputName=ForwardVector, previewValue=<0 1 0 0>}
  n2 = FunctionInput(Preview=Constant3Vector()) {bUsePreviewValueAsDefault=true, inputName=MaskOffset, previewValue=<0 1 0 0>}
  n3 = MaterialFunctionCall!(
   Vector to Transform=CameraPositionWS()
   Z Vector=Transform(None=Normalize(VectorInput=@n1)) {transformSourceType=TRANSFORMSOURCE_Local}
   Center Location=Add(A=ObjectPositionWS(), B=Transform(None=@n2) {transformSourceType=TRANSFORMSOURCE_Local})
  ) {materialFunction=/Engine/Functions/Engine_MaterialFunctions02/Math/TransformToZVector}
  FunctionOutput(None=Clamp(None=ComponentMask(None=@n3.Result) {b=true})) {outputName=CameraPositionMask, sortPriority=4}
  n4 = WorldPosition()
  n5 = ObjectPositionWS()
  n6 = Add(
   A=Transform(None=Subtract(A=@n4.XYZ, B=@n5)) {transformSourceType=TRANSFORMSOURCE_World, transformType=TRANSFORM_Local}
   B=FunctionInput(Preview=Constant()) {bUsePreviewValueAsDefault=true, inputName=Offset, inputType=FunctionInput_Scalar, sortPriority=5}
  )
  n7 = FunctionInput(Preview=Constant()) {bUsePreviewValueAsDefault=true, inputName=Falloff, inputType=FunctionInput_Scalar, sortPriority=4, previewValue=<1 1 1 1>}
  n8 = Clamp(None=Multiply(A=ComponentMask(None=@n6) {g=true}, B=@n7))
  FunctionOutput(None=@n8) {description=Gradient, outputName=Debug, sortPriority=5}
  n9 = FunctionInput(Preview=Constant()) {bUsePreviewValueAsDefault=true, inputName=BiasMax, inputType=FunctionInput_Scalar, sortPriority=3, previewValue=<1 1 1 1>}
  n10 = FunctionInput(Preview=Constant()) {bUsePreviewValueAsDefault=true, inputName=BisMin, inputType=FunctionInput_Scalar, sortPriority=3, previewValue=<1 1 1 1>}
  n11 = Add(A=PixelDepth().R, B=LinearInterpolate(A=@n9, B=@n10, Alpha=@n8))
  n12 = ComponentMask(None=SceneDepth().R) {r=true}
  n13 = Constant()
  n14 = Constant() {r=1}
  FunctionOutput(None=Clamp(None=If(A=@n11, B=@n12, "A > B"=@n13, "A == B"=@n14, "A < B"=@n14))) {
   description="Uses depth to remove objects near object that uses this function with a gradient(Min and Max Values)"
   outputName=DepthGradientMask
   sortPriority=3
  }
 ]}
