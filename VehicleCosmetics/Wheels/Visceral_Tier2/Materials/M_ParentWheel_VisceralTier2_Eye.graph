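# Translucent lit parent material for the animated eye on the Visceral Tier 2 vehicle wheel cosmetic, usable on skeletal meshes.
# Panning gradient noise and a radial, cloud-modulated bulge distort the eye texture, which is contrast-boosted and tinted by CustomColor.
# A local-Y gradient drives colored emissive fog, and the base color lerps toward black.
# Sine/cosine time waves create a world position offset wobble, while MF_CylinderVMask produces a hard-edged opacity mask.
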Material /VehicleCosmetics/Wheels/Visceral_Tier2/Materials/M_ParentWheel_VisceralTier2_Eye
graph {
 properties {bAutomaticallySetUsageInEditor=false, bUsedWithSkeletalMesh=true, blendMode=BLEND_Translucent, usedShadingModels=Default Lit}
 nodes [
  n1 = TextureCoordinate() {uTiling=0, vTiling=0}
  n2 = Multiply(A=Time(), B=Constant() {r=2})
  n3 = TextureSample(UVs=Panner(Coordinate=@n1, Time=@n2) {speedX=-0.1, speedY=0.2}) {texture=/VehicleCosmetics/Wheels/_SharedTextures/FX/MowGradient01}
  ComponentMask(None=@n3.G) {g=true}
  n4 = TextureSample(UVs=Panner(Coordinate=@n1, Time=@n2) {speedX=0.12, speedY=-0.05}) {samplerSource=SSM_Wrap_WorldGroupSettings, texture=&MowGradient01}
  ComponentMask(None=@n4.G) {g=true}
  n5 = Subtract(A=WorldPosition().XYZ, B=ObjectPositionWS())
  n6 = Add(
   A=ComponentMask(None=Transform(None=@n5) {transformSourceType=TRANSFORMSOURCE_World, transformType=TRANSFORM_Local}) {g=true}
   B=ScalarParameter() {defaultValue=-24, parameterName=FogOffset}
  )
  n7 = Clamp(None=Multiply(A=OneMinus(None=@n6), B=ScalarParameter() {defaultValue=0.05, parameterName=FogFalloff}))
  NamedRerouteDeclaration(None=@n7) {name="A[pha", nodeColor=<rgba 1 0.5625 0 1>}
  n8 = AppendVector(A=@n3.G, B=@n4.G)
  n9 = Constant() {r=2}
  n10 = ConstantBiasScale(None=Clamp(None=Multiply(A=@n8, B=@n9))) {bias=-0.5, scale=0.15}
  n11 = Add(A=Add(A=TextureCoordinate() {uTiling=2, vTiling=2}, B=Constant() {r=-1}), B=@n10)
  n12 = Clamp(None=OneMinus(None=DotProduct(A=@n11, B=@n11)))
  n13 = Constant() {r=16}
  n14 = Clamp(None=Multiply(A=Power(Base=@n12, Exp=@n13), B=Constant() {r=1}))
  n15 = Panner(Coordinate=TextureCoordinate() {uTiling=0, vTiling=0}) {speedX=0.5, speedY=0.05}
  n16 = Multiply(
   A=Multiply(A=Multiply(A=@n14, B=@n11), B=TextureSample(UVs=@n15) {texture=/VehicleCosmetics/Wheels/_SharedTextures/FX/CloudDistort_T}.G)
   B=Constant() {r=0.36}
  )
  n17 = Add(A=@n10, B=Multiply(A=TextureCoordinate() {uTiling=3, vTiling=3}, B=Constant() {r=-0.35}))
  n18 = Power(
   Base=TextureSample(UVs=Add(A=@n16, B=@n17)) {samplerSource=SSM_Wrap_WorldGroupSettings, texture=/VehicleCosmetics/Wheels/Visceral_Tier2/Textures/T_Visceral_Tier2_Eye_BC}.G
   Exp=Constant() {r=4}
  )
  n19 = Multiply(
   A=VectorParameter() {parameterName=CustomColor, defaultValue=<rgba 1 1 1 1>}.RGB
   B=ScalarParameter() {defaultValue=1, parameterName=Brightnes}
  )
  NamedRerouteDeclaration(None=LinearInterpolate(A=Multiply(A=@n18, B=@n19), B=Constant(), Alpha=NamedRerouteUsage())) {name=BC, nodeColor=<rgba 0.2422 0 1 1>}
  n20 = VectorParameter() {parameterName=FogColor, defaultValue=<rgba 0 0 0 1>}
  n21 = ScalarParameter() {defaultValue=1, parameterName=FogIntensity}
  NamedRerouteDeclaration(None=Multiply(A=@n7, B=Multiply(A=@n20.RGB, B=@n21))) {name=Emissive, nodeColor=<rgba 0.0781 1 0 1>}
  n22 = Time()
  n23 = Multiply(A=Constant() {r=10}, B=@n22)
  n24 = Sine(None=@n23)
  n25 = Cosine(None=@n23)
  n26 = Multiply(A=AppendVector(A=AppendVector(A=@n24, B=@n25), B=@n25), B=Constant() {r=0.075})
  n27 = Sine(None=@n22)
  n28 = Constant()
  n29 = Sine(None=Multiply(A=Constant() {r=10}, B=@n22))
  n30 = AppendVector(A=@n28, B=Multiply(A=Multiply(A=@n29, B=@n27), B=Constant() {r=0.2}))
  NamedRerouteDeclaration(None=Add(A=Multiply(A=@n26, B=@n27), B=AppendVector(A=@n30, B=@n28))) {name=WPO, nodeColor=<rgba 1 0 0.3984 1>}
  n31 = MaterialFunctionCall!(
   MaskOffset
   ForwardVector
   BiasMax=ScalarParameter() {parameterName=BiasMax}
   BisMin=ScalarParameter() {parameterName=BiasMin}
   Falloff=ScalarParameter() {defaultValue=1, parameterName=Fallof}
   Offset=ScalarParameter() {parameterName=Offset}
  ) {materialFunction=&/MF_CylinderVMask}
  n32 = Multiply(A=@n31.DepthGradientMask, B=@n31.DepthGradientMask)
  NamedRerouteDeclaration(None=Clamp(None=Ceil(None=@n32))) {name=Opacity, nodeColor=<rgba 1 0 0 1>}
  NamedRerouteUsage()
  NamedRerouteUsage()
  NamedRerouteUsage()
  NamedRerouteUsage()
 ]
}
