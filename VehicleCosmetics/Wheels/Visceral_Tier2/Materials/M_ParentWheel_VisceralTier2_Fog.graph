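# Translucent, lit parent material for the fog on Visceral Tier 2 vehicle wheels, flagged for skeletal meshes.
# An object-relative, local-space Y gradient, shaped by offset and falloff parameters, scales the fog color and intensity for an emissive term.
# A base color term blends tinted CustomColor toward black.
# MF_CylinderVMask's depth gradient is squared and rounded up into a hard mask, then scaled by an Opacity parameter.
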Material /VehicleCosmetics/Wheels/Visceral_Tier2/Materials/M_ParentWheel_VisceralTier2_Fog
graph {
 properties {bAutomaticallySetUsageInEditor=false, bUsedWithSkeletalMesh=true, blendMode=BLEND_Translucent, usedShadingModels=Default Lit}
 nodes [
  n1 = Subtract(A=WorldPosition().XYZ, B=ObjectPositionWS())
  n2 = Add(
   A=ComponentMask(None=Transform(None=@n1) {transformSourceType=TRANSFORMSOURCE_World, transformType=TRANSFORM_Local}) {g=true}
   B=ScalarParameter() {defaultValue=-24, parameterName=FogOffset}
  )
  n3 = Clamp(None=Multiply(A=OneMinus(None=@n2), B=ScalarParameter() {defaultValue=0.05, parameterName=FogFalloff}))
  NamedRerouteDeclaration(None=@n3) {name="A[pha", nodeColor=<rgba 1 0.5625 0 1>}
  n4 = VectorParameter() {parameterName=CustomColor, defaultValue=<rgba 1 1 1 1>}
  n5 = ScalarParameter() {defaultValue=1, parameterName=Brightnes}
  NamedRerouteDeclaration(None=LinearInterpolate(A=Multiply(A=@n4.RGB, B=@n5), B=Constant(), Alpha=NamedRerouteUsage())) {name=BC, nodeColor=<rgba 0.2422 0 1 1>}
  n6 = VectorParameter() {parameterName=FogColor, defaultValue=<rgba 0 0 0 1>}
  n7 = ScalarParameter() {defaultValue=1, parameterName=FogIntensity}
  NamedRerouteDeclaration(None=Multiply(A=@n3, B=Multiply(A=@n6.RGB, B=@n7))) {name=Emissive, nodeColor=<rgba 0.0781 1 0 1>}
  n8 = MaterialFunctionCall!(
   MaskOffset
   ForwardVector
   BiasMax=ScalarParameter() {parameterName=BiasMax}
   BisMin=ScalarParameter() {parameterName=BiasMin}
   Falloff=ScalarParameter() {defaultValue=1, parameterName=Fallof}
   Offset=ScalarParameter() {parameterName=Offset}
  ) {materialFunction=&/MF_CylinderVMask}
  n9 = Ceil(None=Multiply(A=@n8.DepthGradientMask, B=@n8.DepthGradientMask))
  NamedRerouteDeclaration(None=Multiply(A=Clamp(None=@n9), B=ScalarParameter() {defaultValue=1, parameterName=Opacity})) {name=Opacity, nodeColor=<rgba 1 0 0 1>}
  NamedRerouteUsage()
  NamedRerouteUsage()
  NamedRerouteUsage()
 ]
}
