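# Translucent lit parent material for the teeth on the Visceral Tier 2 skeletal-mesh vehicle wheel cosmetic.
# Base color blends AO-multiplied texture or Painted-switch custom color with brightening and darkening, alongside a normal map and tunable colored fog emissive built around a local-Y gradient.
# Vertex RGB channels weight sine and cosine waves that animate world position offset around the object pivot.
# A cylinder mask feeds the opacity blend at zero weight.
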Material /VehicleCosmetics/Wheels/Visceral_Tier2/Materials/M_ParentWheel_VisceralTier2_Teeth
graph {
 properties {
  bAutomaticallySetUsageInEditor=false
  bUsedWithSkeletalMesh=true
  blendMode=BLEND_Translucent
  previewMesh=/VehicleCosmetics/Wheels/Visceral_Tier2/Meshes/SK_Wheel_Visceral_Tier2
  usedShadingModels=Default Lit
 }
 nodes [
  n1 = MaterialFunctionCall!(
   MaskOffset
   ForwardVector
   BiasMax=ScalarParameter() {defaultValue=-4, parameterName=BiasMax}
   BisMin=ScalarParameter() {parameterName=BiasMin}
   Falloff=ScalarParameter() {defaultValue=1, parameterName=Fallof}
   Offset=ScalarParameter() {parameterName=Offset}
  ) {materialFunction=&/MF_CylinderVMask}
  n2 = Clamp(None=Ceil(None=Multiply(A=@n1.DepthGradientMask, B=@n1.DepthGradientMask)))
  n3 = NamedRerouteUsage()
  Multiply(A=@n2, B=@n3)
  n4 = Transform(None=Subtract(A=WorldPosition().XYZ, B=ObjectPositionWS())) {transformSourceType=TRANSFORMSOURCE_World, transformType=TRANSFORM_Local}
  n5 = OneMinus(None=Add(A=ComponentMask(None=@n4) {g=true}, B=ScalarParameter() {defaultValue=-24, parameterName=FogOffset}))
  n6 = ScalarParameter() {defaultValue=0.05, parameterName=FogFalloff}
  NamedRerouteDeclaration(None=Clamp(None=Multiply(A=@n5, B=@n6))) {name="A[pha", nodeColor=<rgba 1 0.5625 0 1>}
  n7 = TextureSampleParameter2D() {
   parameterName=Mask
   samplerSource=SSM_Wrap_WorldGroupSettings
   samplerType=SAMPLERTYPE_LinearColor
   texture=/VehicleCosmetics/Wheels/Visceral_Tier1/Textures/T_Visceral_Tier1_Teeth_Alpha
  }
  NamedRerouteDeclaration(None=@n7.R) {name=Mask_RC, nodeColor=<rgba 0 0.0859 1 1>}
  n8 = VectorParameter() {parameterName=FogColor, defaultValue=<rgba 1 0.2428 0 1>}
  n9 = ScalarParameter() {defaultValue=1, parameterName=FogIntensity}
  NamedRerouteDeclaration(None=Multiply(A=NamedRerouteUsage(), B=Multiply(A=@n8.RGB, B=@n9))) {name=Emissive, nodeColor=<rgba 0.0781 1 0 1>}
  n10 = VectorParameter() {parameterName=CustomColor, defaultValue=<rgba 1 1 1 1>}
  n11 = TextureSampleParameter2D() {
   parameterName=AO
   samplerSource=SSM_Wrap_WorldGroupSettings
   samplerType=SAMPLERTYPE_LinearColor
   texture=/VehicleCosmetics/Wheels/Visceral_Tier2/Textures/T_Visceral_Tier2_Teeth_AO
  }
  n12 = Constant() {r=0.2}
  n13 = Desaturation(
   None=Multiply(
    A=@n11.RGB
    B=TextureSample() {samplerSource=SSM_Wrap_WorldGroupSettings, texture=/VehicleCosmetics/Wheels/Visceral_Tier2/Textures/T_Visceral_Tier2_Teeth_BC}.RGB
   )
   Fraction=@n12
  ) {luminanceFactors=<rgba 0.3 0.59 0.11 0>}
  n14 = StaticSwitchParameter(True=Desaturation(None=Multiply(A=@n10.RGB, B=@n11.RGB), Fraction=@n12) {luminanceFactors=<rgba 0.3 0.59 0.11 0>}, False=@n13) {parameterName=Painted}
  n15 = Multiply(A=@n13, B=Constant() {r=1.7})
  n16 = NamedRerouteUsage()
  NamedRerouteDeclaration(None=LinearInterpolate(A=LinearInterpolate(A=@n14, B=@n15, Alpha=@n16), B=Constant(), Alpha=NamedRerouteUsage())) {name=BC, nodeColor=<rgba 0.4062 1 0 1>}
  NamedRerouteDeclaration(None=LinearInterpolate(A=@n3, B=@n2, Alpha=Constant())) {name=Opacity, nodeColor=<rgba 1 0 0 1>}
  n17 = VertexColor()
  n18 = Multiply(A=Desaturation(None=@n17) {luminanceFactors=<rgba 0.3 0.59 0.11 0>}, B=Constant() {r=256})
  n19 = ScalarParameter() {defaultValue=-0.075, parameterName=Offset_WPO}
  n20 = Time()
  n21 = Add(A=Multiply(A=Sine(None=@n20) {period=0.5}, B=@n17.R), B=Multiply(A=Cosine(None=@n20), B=@n17.G))
  n22 = Multiply(A=Sine(None=@n20), B=@n17.B)
  n23 = Multiply(
   A=Subtract(
    A=WorldPosition().XYZ
    B=MaterialFunctionCall() {materialFunction=/VehicleCosmetics/Wheels/Visceral_Tier1/Materials/ObjectPivotPoint}.ObjectPivotLoacation
   )
   B=ConstantBiasScale(None=Add(A=@n21, B=@n22)) {bias=2}
  )
  NamedRerouteDeclaration(None=Multiply(A=Clamp(None=@n18), B=Multiply(A=@n19, B=@n23))) {name=WPO, nodeColor=<rgba 0 0.6484 1 1>}
  NamedRerouteUsage()
  NamedRerouteUsage()
  NamedRerouteUsage()
  NamedRerouteUsage()
  NamedRerouteUsage()
  TextureSampleParameter2D() {parameterName=Normal, samplerType=SAMPLERTYPE_Normal, texture=/VehicleCosmetics/Wheels/Visceral_Tier1/Textures/T_Visceral_Tier1_Teeth_N}
 ]
}
